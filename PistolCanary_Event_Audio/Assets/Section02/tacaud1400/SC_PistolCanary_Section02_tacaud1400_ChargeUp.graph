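# Sound cue for a tentacle attack charge-up in Section02 of the PistolCanary event, classed under Fort_Events.
# A single wave player feeds the output at reduced volume, spatialized through the tentacle-tip attenuation asset with a very large max distance.
# Distance-scaled linear bus sends feed the weapon volume and music low-pass sidechain buses, with send level rising between near and far thresholds.
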SoundCue /PistolCanary_Event_Audio/Assets/Section02/tacaud1400/SC_PistolCanary_Section02_tacaud1400_ChargeUp
root=WavePlayer_0
nodes [{id=WavePlayer_0, properties {soundWaveAssetPtr=/PistolCanary_Event_Audio/Assets/Section02/wav/PistolCanary_Tentacle_Attk_1000_ChargeUp}}]
properties {
 attenuationSettings=/PistolCanary_Audio/Assets/Settings/ATT_PistolCanary_Tentacle_Tip
 maxDistance=100000
 soundClassObject=&Fort_Events
 volumeMultiplier=0.6
 busSends [
  {
   audioBus=&AB_Sidechain_FortWeapon_Vol
   maxSendDistance=20000
   maxSendLevel=0.4
   minSendDistance=9000
   minSendLevel=0
   sendLevel=0.3
   soundSourceBus=null
   sourceBusSendLevelControlMethod=Linear
  }
  {
   audioBus=/Game/Sounds/Modulation/EnvelopeFollowers/LPF/AB_SideChain_FortMusic_LPF
   maxSendDistance=20000
   maxSendLevel=0.5
   minSendDistance=9000
   minSendLevel=0
   sendLevel=0.3
   soundSourceBus=null
   sourceBusSendLevelControlMethod=Linear
  }
 ]
}
