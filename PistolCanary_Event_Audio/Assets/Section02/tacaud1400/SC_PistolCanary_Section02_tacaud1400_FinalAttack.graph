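# Sound cue for the final tentacle attack shot in Section 02 of the PistolCanary event.
# A single wave player feeds the output, using the tentacle-tip attenuation and the Fort_Events sound class at slightly reduced volume.
# Distance-scaled linear bus sends drive sidechains for weapon volume, music volume and music high-pass filtering, with send levels rising from near to far.
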SoundCue /PistolCanary_Event_Audio/Assets/Section02/tacaud1400/SC_PistolCanary_Section02_tacaud1400_FinalAttack
root=WavePlayer_0
nodes [{id=WavePlayer_0, properties {soundWaveAssetPtr=/PistolCanary_Event_Audio/Assets/Section02/wav/PistolCanary_Tentacle_Attk_1000_FinalShot}}]
properties {
 attenuationSettings=/PistolCanary_Audio/Assets/Settings/ATT_PistolCanary_Tentacle_Tip
 maxDistance=100000
 soundClassObject=&Fort_Events
 volumeMultiplier=0.8
 busSends [
  {
   audioBus=&AB_Sidechain_FortWeapon_Vol
   maxSendDistance=20000
   maxSendLevel=0.6
   minSendDistance=1300
   minSendLevel=0.1
   sendLevel=0.2
   soundSourceBus=null
   sourceBusSendLevelControlMethod=Linear
  }
  {
   audioBus=&AB_Sidechain_FortMusic_Vol
   maxSendDistance=20000
   minSendDistance=1300
   minSendLevel=0
   sendLevel=1
   soundSourceBus=null
   sourceBusSendLevelControlMethod=Linear
  }
  {
   audioBus=/Game/Sounds/Modulation/EnvelopeFollowers/HPF/AB_SideChain_FortMusic_HPF
   maxSendDistance=20000
   maxSendLevel=0.5
   minSendDistance=1000
   minSendLevel=0
   sendLevel=1
   soundSourceBus=null
   sourceBusSendLevelControlMethod=Linear
  }
 ]
}
